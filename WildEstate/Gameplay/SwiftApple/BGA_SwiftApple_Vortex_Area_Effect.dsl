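; Swift Apple vent vortex area effect actor that removes itself in playlists tagged as LTM, noted for the Infiltration playlist.
; On BeginPlay the authority waits for playlist data, then destroys the actor if the playlist's context tags match the LTM tag.
; The construction script resets the geyser cone component's relative rotation.
; Overlap and Tick events are empty.
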(event EventBeginPlay
  (|CallParentFunction)
  ;[1 1 1] In Infiltration playlist, remove vent vortex
  (Networking|SwitchHasAuthority
    (:Authority
      (bind (_gamestate _playlist _playlistcontexttags) (Class|FortAsyncActionCurrentPlaylistReady|PlaylistDataReadyAsync)
        (:OnReady
          (if (GameplayTags|HasAnyTags _playlistcontexttags (Variables|Default|GetLTMTag) true)
            (Actor|DestroyActor)))))))

(event Collision|EventActorBeginOverlap
  (params (OtherActor :type Actor%20Object%20Reference)))

(event EventTick
  (params (DeltaSeconds :type Float%20%28single-precision%29)))

(fn ConstructionScript
  (|CallParentFunction)
  (Transformation|SetRelativeRotation (Variables|Default|GetS_FX_GeyserCone) (Math|Rotator|MakeRotator 0.0) false false))
