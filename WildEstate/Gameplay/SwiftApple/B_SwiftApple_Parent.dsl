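; SwiftApple parent Blueprint is a tire-style bounce pad that launches whatever overlaps its jump collision.
; Player and AI pawns, screened for launched or DBNO state, launch upward and forward with a gameplay effect and fall-damage immunity; players also get pending skydive.
; Other actors try projectile redirection, excluding petrol and pickups, and hoverboard launch.
; A ceiling check sets bounce height, with wobble, sound and force feedback accompanying bounces.
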(event OnComponentBeginOverlap(Jump_Collision)
  (params (OverlappedComponent :type Primitive%20Component%20Object%20Reference) (OtherActor :type Actor%20Object%20Reference) (OtherComp :type Primitive%20Component%20Object%20Reference) (OtherBodyIndex :type Integer) (bFromSweep :type Boolean) (SweepResult :type Hit%20Result%20Structure%20%28by%20ref%29))
  (bind (_bblockinghit _binitialoverlap _time _distance _location _impactpoint) (Collision|BreakHitResult SweepResult))
  (bind _overlappawn (Variables|Default|GetOverlapPawn))
  (bind _overlappawn_2 (Variables|Default|GetOverlapPawn))
  (bind _overlappawn_3 (Variables|Default|GetOverlapPawn))
  (Variables|Default|SetBounceLocation (Transformation|GetActorLocation))
  (CallFunction|MakeWobbleEffect _impactpoint)
  (bind _asfort_player_pawn (Utilities|Casting|CastToFortPlayerPawn OtherActor)
    (:then
      (Variables|Default|SetOverlapPawn _asfort_player_pawn))
    (:CastFailed
      (bind _asfort_aipawn (Utilities|Casting|CastToFortAIPawn OtherActor)
        (:then
          (Variables|Default|SetOverlapPawn _asfort_aipawn))
        (:CastFailed
          (CallFunction|LaunchProjectile OtherActor)
          ;[1 1 1] Ceiling check
          (bind _success (CallFunction|CeilingCheck))
          (if _success
            (Variables|Default|SetBounceHeight 600)
            (Variables|Default|SetJumpHeight 100)
            (else
              (Variables|Default|SetBounceHeight 2e+03)
              (Variables|Default|SetJumpHeight 600)))
          (Variables|Default|SetFinalBounceJumpHeight (Variables|Default|GetBounceHeight))
          (bind _success_2 (Vehicles|LaunchHoverboard OtherActor (Math|Vector|MakeVector 0.0 0.0 (Variables|Default|GetFinalBounceJumpHeight)) "TireLaunch"))
          (if _success_2
            (Networking|ForceNetUpdate)
            (CallFunction|PlayBounceSound))))))
  (Utilities|IsValid _overlappawn_2
    (:Is%20Valid
      (if (or (GameplayTags|HasMatchingGameplayTag (Ability|GetAbilitySystemComponent _overlappawn_2) "(TagName=\"Gameplay.Action.Player.Launched\")") (Variables|/Script/FortniteGame.FortPawn|GetbIsDBNO _overlappawn_2))
        _
        (else
          _)))
    (:Is%20Not%20Valid
      _))
  (bind _asfort_player_controller (Utilities|Casting|CastToFortPlayerController (Pawn|GetController _overlappawn_2))
    (:then
      (Variables|Default|SetOverlapController _asfort_player_controller))
    (:CastFailed
      _))
  ;[1 1 1] Ceiling check
  (bind _success_3 (CallFunction|CeilingCheck))
  (if _success_3
    (Variables|Default|SetBounceHeight 600)
    (Variables|Default|SetJumpHeight 100)
    (else
      (Variables|Default|SetBounceHeight (Variables|Default|GetNoCeilingBounceHeight))
      (Variables|Default|SetJumpHeight 600)))
  (Variables|Default|SetFinalBounceJumpHeight (Variables|Default|GetBounceHeight))
  (bind _asfort_player_pawn_2 (Utilities|Casting|CastToFortPlayerPawn (Variables|Default|GetOverlapPawn))
    (:then
      (RemoteControlPawn|ForceDetachfromRemoteControlledPawn _asfort_player_pawn_2)
      (AthenaMovement|SetPendingSkydiveLaunch _asfort_player_pawn_2 true))
    (:CastFailed
      _))
  (bind (_returnvalue_x _returnvalue_y) (Math|Vector|BreakVector (Transformation|GetActorForwardVector _asfort_player_pawn_2)))
  (Character|LaunchCharacter _overlappawn_3 (Math|Vector|MakeVector (* _returnvalue_x (Variables|Default|GetBounceXYVelocityMofifier)) (* _returnvalue_y (Variables|Default|GetBounceXYVelocityMofifier)) (Variables|Default|GetFinalBounceJumpHeight)) false true)
  (Ability|SendGameplayEventtoActor _overlappawn_3 "(TagName=\"Event.Tire.Bounce\")")
  (GameplayEffects|ApplyGameplayEffectToSelf :self (Ability|GetAbilitySystemComponent _overlappawn) :GameplayEffectClass (Variables|Default|GetGE_PlayerLaunched) :Level 0.0)
  (Variables|/Script/FortniteGame.FortPawn|SetbIgnoreNextFallingDamage true _overlappawn)
  (Utilities|FlowControl|Sequence
    (:then_0
      (bind _overlapcontroller (Variables|Default|GetOverlapController)
        (:then
          (Game|Feedback|LocalPlayForceFeedback _overlapcontroller (Variables|Default|GetLaunchForceFeedback) false "TireLaunch"))
        (:else
          _))
      (Variables|Default|SetOverlapPawn 0)
      (Variables|Default|SetOverlapController 0))
    (:then_1
      (Networking|ForceNetUpdate)
      (CallFunction|PlayBounceSound))))

(fn LaunchProjectile
  (params (Actor :type Actor%20Object%20Reference))
  (bind _returnvalue (Actor|GetComponentbyClass Actor "/Script/Engine.ProjectileMovementComponent"))
  (bind (_asfort_projectile_petrol _bsuccess) (Utilities|Casting|CastToFortProjectilePetrol Actor))
  (bind (_asfort_pickup _bsuccess_2) (Utilities|Casting|CastToFortPickup Actor))
  (if (Math|Boolean|ANDBoolean (or (Utilities|IsValid _returnvalue :!class /Script/Engine.KismetSystemLibrary) (Utilities|IsValid (Actor|GetComponentbyClass Actor "/Script/FortniteGame.FortProjectileMovementComponent") :!class /Script/Engine.KismetSystemLibrary)) (not _bsuccess) (not _bsuccess_2) (Variables|Default|GetBBouncingProjectile))
    (Variables|/Script/Engine.MovementComponent|SetVelocity (Utilities|GetNonPawnVelocity Actor) _returnvalue)))

(fn ConstructionScript
  (|CallParentFunction))
